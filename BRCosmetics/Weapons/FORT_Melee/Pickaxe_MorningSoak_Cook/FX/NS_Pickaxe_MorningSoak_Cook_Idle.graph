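# Idle Niagara effect for the MorningSoak Cook pickaxe: a flickering green glow attached to the weapon.
# Its single LightGlow emitter burst-spawns particles that follow a socket through the system's Socket Reader, with alignment taken from socket rotation.
# Two sprite renderers draw from glow and fake-light attributes, both pulsed by curves, with fake-light color also scaled by a system time-of-day value.
# Light attributes are written, but no light renderer is present.
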NiagaraSystem /BRCosmetics/Weapons/FORT_Melee/Pickaxe_MorningSoak_Cook/FX/NS_Pickaxe_MorningSoak_Cook_Idle
dependencies {
 usedDataInterfaces [&NiagaraDataInterfaceCurve, &NiagaraDataInterfaceColorCurve, &NiagaraDataInterfaceSocketReader]
 usedDynamicInputs []
 usedModules [
  &self:SystemSpawnScript.SystemScriptSource.SystemScriptGraph.NiagaraNodeAssignment_0.SetVariables_#1
  &SystemState
  &self:SystemSpawnScript.SystemScriptSource.SystemScriptGraph.NiagaraNodeAssignment_1.SetVariables_#2
 ]
}
emitters {LightGlow {
  emitterUpdateScript [
   &EmitterState(Inactive Response=::NewEnumerator1, Life Cycle Mode=::NewEnumerator1, Loop Duration=1, UseLoopDelay=false)
   &SpawnBurst_Instantaneous(Use Spawn Probability=false, UseLoopCountLimit=false)
  ]
  particleSpawnScript [
   &InitializeParticle(
    Sprite Rotation Mode=::NewEnumerator0
    Sprite Size Mode=::NewEnumerator3
    Uniform Sprite Size=18
    UsePositionOffset=true
    Write Material Random=true
   )
   &CameraOffset(Camera Offset Amount=3)
   /SetVariables(Particles.Size_FakeLight:Vector2f=&Vector2DFromFloat(Value=20)) {isSetParameters=true}
  ]
  particleUpdateScript [
   &ParticleState()
   &SocketLocation(SamplePosOffset=true, Socket Reader=$System.Socket Reader, Socket Sampling Mode=::NewEnumerator2, VelocitySpeedLimit=false)
   &FN_Sprite_Attributes(
    Camera Offset=15
    Color=&MultiplyLinearColor(
     Linear Color=<rgba 0 10 0.454 1>
     Scale Factor=&MakeVector4(
      X=1
      Y=1
      Z=1
      W=&Multiply_Float(
       A=0.5
       B=&FloatFromCurve(
        FloatCurve=DataInterface() {propertyValues {Curve {
           defaultValue=0.5715
           keys [
            {tangentMode=RCTM_User, time=0, value=1}
            <at 0.2907 1>
            <at 0.3314 0.3>
            <at 0.3722 1>
            <at 0.701 1>
            <at 0.7507 0.5673>
            <at 0.7976 1>
            <at 0.8826 1>
            <at 0.9663 0.3>
            {tangentMode=RCTM_User, time=1, value=1}
           ]
          }}}
        Scale Curve=0.3
       )
      )
     )
    )
    Create Camera Offset Attribute=true
    Create Color Attribute=true
    Create Position Attribute=true
    Create SpriteSize Attribute=true
    SpriteSize=&Multiply_Vector2D_ByFloat(Float=1, Vector2D=&Vector2DFromFloat(Value=18))
    Write to Module Namespace=true
   )
   &SpriteFacingAndAlignment(
    Sprite Alignment=&RotateVectorByQuaternion(Vector To Rotate=<1 0 0>, Quaternion=$Output.SocketLocation.SampledRotation)
    WriteAlignment=true
    WriteFacing=true
   ) {module=SpriteFacingAndAlignment001}
   &FN_Light_Attributes(
    Color=<rgba 1.9359 6 0.5625 1>
    Create Color Attribute=true
    Create Exponent Attribute=true
    Create Radius Attribute=true
    Exponent=2
    Radius=30
    Scale Color=&FNTimeOfDayScalar(Night Scale=0.8, Morning Scale=0.9, Evening Scale=0.85)
   )
   /SetVariables(Particles.Color_FakeLight:LinearColor=&Multiply_LinearColorByFloat(
     Linear Color=&ColorFromCurve(ColorCurve=DataInterface() {propertyValues {
        AlphaCurve {keys [<at 0.0111 1>,<at 0.9846 1>]}
        BlueCurve {keys [<at 0 0.0103 RCIM_Cubic>]}
        GreenCurve {keys [<at 0 0.3021 RCIM_Cubic>]}
        RedCurve {keys [<at 0 0 RCIM_Cubic>]}
       }})
     Float=&Multiply_Float(
      A=&FloatFromCurve(FloatCurve=DataInterface() {propertyValues {Curve {
          defaultValue=0.5715
          keys [
           {tangentMode=RCTM_User, time=0, value=1}
           <at 0.2907 1>
           <at 0.3314 0.3>
           <at 0.3722 1>
           <at 0.701 1>
           <at 0.7507 0.5673>
           <at 0.7976 1>
           <at 0.8826 1>
           <at 0.9663 0.3>
           {tangentMode=RCTM_User, time=1, value=1}
          ]
         }}})
      B=$System.TODFloat
     )
    )) {isSetParameters=true}
  ]
  rendererClasses [&NiagaraSpriteRendererProperties]
  renderers [
   {rendererClass=&NiagaraSpriteRendererProperties}
   {rendererClass=&NiagaraSpriteRendererProperties, rendererIndex=1}
  ]
 }}
summary {
 systemName=NS_Pickaxe_MorningSoak_Cook_Idle
 emitters [{emitterName=LightGlow, rendererClasses [&NiagaraSpriteRendererProperties]}]
 userVariables [
  User.bIsFrontEnd {type=NiagaraBool}
  User.bIsFrontEndPreview {type=NiagaraBool}
 ]
}
