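# Niagara swing trail for the MorningSoak Cook pickaxe, built from two ribbon emitters that use the shared AnimTrail functions to spawn, initialize and update from bones or sockets.
# LightTrail is a wide, short-lived green ribbon that darkens along a color curve, with alpha pulsed by emitter looped age.
# ToonTrail is a much narrower ribbon whose alpha combines a curve with System.GeneralSwingFade.
# Both tint their color by the System.TOD value.
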NiagaraSystem /BRCosmetics/Weapons/FORT_Melee/Pickaxe_MorningSoak_Cook/FX/NS_Pickaxe_MorningSoak_Cook_Trail
dependencies {
 usedDataInterfaces [&NiagaraDataInterfaceColorCurve, &NiagaraDataInterfaceCurve, &NiagaraDataInterfaceSocketReader]
 usedDynamicInputs []
 usedModules [
  &self:SystemSpawnScript.SystemScriptSource.SystemScriptGraph.NiagaraNodeAssignment_3.SetVariables_#1
  &self:SystemSpawnScript.SystemScriptSource.SystemScriptGraph.NiagaraNodeAssignment_0.SetVariables_#2
  &self:SystemSpawnScript.SystemScriptSource.SystemScriptGraph.NiagaraNodeAssignment_1.SetVariables_#3
  &SystemState
 ]
}
emitters {
 LightTrail {
  emitterSpawnScript [&FN_RibbonIDOffsetUtility()]
  emitterUpdateScript [
   &EmitterState()
   &FN_SpawnAnimTrail()
  ]
  particleSpawnScript [
   &FN_InitializeAnimTrailBoneIndecies(First Bone or Socket Index=2, Second Bone or Socket Index=2)
   &FN_InitializeAnimTrail(AnimTrailWidth=10, Lifetime=0.28)
   /SetVariables(Particles.Color=&MultiplyLinearColor(Linear Color=$Particles.Color, Scale Factor=&MakeVector4_FromLinearColor(LinearColor=$System.TOD))) {isSetParameters=true}
  ]
  particleUpdateScript [
   &FN_UpdateAnimTrail() {module=FN_UpdateAnimTrail001}
   &ScaleColor(
    Scale Alpha=&FloatFromCurve(
     FloatCurve=DataInterface() {propertyValues {Curve {
        defaultValue=1
        keys [
         {arriveTangent=-0.1259, interpMode=RCIM_Cubic, leaveTangent=-0.1259, tangentMode=RCTM_User, time=0, value=0}
         {arriveTangent=-8.3221, interpMode=RCIM_Cubic, leaveTangent=-8.3221, tangentMode=RCTM_User, time=0.0378, value=0.9}
         {interpMode=RCIM_Cubic, tangentMode=RCTM_User, time=0.3, value=0}
        ]
       }}}
     CurveIndex=$Emitter.LoopedAge
     Scale Curve=0.4
    )
    Scale RGB=&MakeVectorFromLinearColorRGB(Color=&ColorFromCurve(ColorCurve=DataInterface() {propertyValues {
        AlphaCurve {keys [<at 0 1 RCIM_Cubic>,<at 0.5977 0 RCIM_Cubic>]}
        BlueCurve {keys [<at 0 0.0908 RCIM_Cubic>,<at 0.1308 0.0435>,<at 0.4059 0.0105>,<at 0.6036 0.0066>,<at 0.8264 0.0012>]}
        GreenCurve {keys [<at 0 2 RCIM_Cubic>,<at 0.1308 0.9577>,<at 0.4059 0.2313>,<at 0.6036 0.1458>,<at 0.8264 0.026>]}
        RedCurve {keys [<at 0 0 RCIM_Cubic>,<at 0.1308 0>,<at 0.4059 0>,<at 0.6036 0>,<at 0.8264 0>]}
       }}))
    ScaleA=true
    ScaleRGB=true
   )
   &SolveForcesAndVelocity(Clamp Velocity=false, Limit Acceleration=false)
  ]
  rendererClasses [&NiagaraRibbonRendererProperties]
  renderers [{rendererClass=&NiagaraRibbonRendererProperties}]
 }
 ToonTrail {
  emitterSpawnScript [&FN_RibbonIDOffsetUtility()]
  emitterUpdateScript [
   &EmitterState()
   &FN_SpawnAnimTrail()
  ]
  particleSpawnScript [
   &FN_InitializeAnimTrailBoneIndecies()
   &FN_InitializeAnimTrail(AnimTrailWidth=1.4, Lifetime=0.3)
  ]
  particleUpdateScript [
   &SolveForcesAndVelocity(Clamp Velocity=false, Limit Acceleration=false)
   &FN_UpdateAnimTrail() {module=FN_UpdateAnimTrail001}
   &ScaleColor(
    Scale Alpha=&Multiply_Float(
     A=$System.GeneralSwingFade
     B=&FloatFromCurve(
      FloatCurve=DataInterface() {propertyValues {Curve {keys [{arriveTangent=1.0098, interpMode=RCIM_Cubic, leaveTangent=1.0098, tangentMode=RCTM_User, time=0.3, value=1}, <at 0.4 0 RCIM_Cubic>]}}}
      Scale Curve=0.9
     )
    )
    Scale RGB=&MakeVectorFromLinearColorRGB(Color=$System.TOD)
    ScaleA=true
    ScaleRGB=true
   )
  ]
  rendererClasses [&NiagaraRibbonRendererProperties]
  renderers [{rendererClass=&NiagaraRibbonRendererProperties}]
 }
}
summary {
 systemName=NS_Pickaxe_MorningSoak_Cook_Trail
 emitters [
  {emitterName=LightTrail, rendererClasses [&NiagaraRibbonRendererProperties]}
  {emitterName=ToonTrail, rendererClasses [&NiagaraRibbonRendererProperties]}
 ]
 userVariables [
  User.bIsFrontEnd {type=NiagaraBool}
  User.bIsFrontEndPreview {type=NiagaraBool}
  User.RibbonIDOffset {type=NiagaraInt32}
 ]
}
