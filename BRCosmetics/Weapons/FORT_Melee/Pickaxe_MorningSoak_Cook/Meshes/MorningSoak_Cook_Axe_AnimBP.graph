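# Post-process style Animation Blueprint for the MorningSoak Cook pickaxe, adding fixed bone offsets and secondary physics motion to an incoming pose.
# Five chained Modify Bone nodes additively nudge the two-handed handle and several wire bones in bone space.
# Three Anim Dynamics nodes then simulate tape1, wires_up and wires_dwn with damped springs, gravity and tight limits, the two wire bodies also sliding linearly.
# The EventGraph holds only unconnected stub nodes, so no gameplay logic drives it.
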AnimBlueprint /BRCosmetics/Weapons/FORT_Melee/Pickaxe_MorningSoak_Cook/Meshes/MorningSoak_Cook_Axe_AnimBP
target_skeleton=&/MorningSoak_Cook_Axe_Skeleton
graphs [
 AnimGraph {nodes [
   AnimGraphNode_SubInput_0 = Input Pose()
   n1 = "Transform (Modify) Bone - Bone: Melee_TwoHanded_Handle"(
    ComponentPose=Local To Component(LocalPose=@AnimGraphNode_SubInput_0)
    Alpha=1
    Translation="-0.000,-0.500,0.450"
    Rotation="0,0,0"
    Scale="1,1,1"
   ) {
    rotationMode=BMM_Additive
    rotationSpace=BCS_BoneSpace
    translationMode=BMM_Additive
    translationSpace=BCS_BoneSpace
    boneToModify {boneName=Melee_TwoHanded_Handle}
    translation {y=-0.5, z=0.45}
   }
   n2 = "Transform (Modify) Bone - Bone: wires_up_base"(ComponentPose=@n1, Alpha=1, Translation="-0.000,-0.243,-0.220", Rotation="-0.209,-0.5,0", Scale="1,1,1") {
    rotationMode=BMM_Additive
    rotationSpace=BCS_BoneSpace
    translationMode=BMM_Additive
    translationSpace=BCS_BoneSpace
    boneToModify {boneName=wires_up_base}
    rotation {pitch=-0.209, yaw=-0.5}
    translation {y=-0.243, z=-0.22}
   }
   n3 = "Transform (Modify) Bone - Bone: wires_dwn_base"(ComponentPose=@n2, Alpha=1, Translation="-0.078,-0.103,0.443", Rotation="0,0,0", Scale="1,1,1") {
    scaleMode=BMM_Additive
    scaleSpace=BCS_BoneSpace
    translationMode=BMM_Additive
    translationSpace=BCS_BoneSpace
    boneToModify {boneName=wires_dwn_base}
    translation=<-0.078 -0.103 0.443>
   }
   n4 = "Transform (Modify) Bone - Bone: wires_cor_01"(ComponentPose=@n3, Alpha=1, Translation="-1.878,0.374,1.120", Rotation="0,0,0", Scale="1,1,1") {
    scaleMode=BMM_Additive
    scaleSpace=BCS_BoneSpace
    translationMode=BMM_Additive
    translationSpace=BCS_BoneSpace
    boneToModify {boneName=wires_cor_01}
    translation=<-1.878 0.374 1.12>
   }
   n5 = "Transform (Modify) Bone - Bone: wires_cor_02"(ComponentPose=@n4, Alpha=1, Translation="0.000,0.144,-0.518", Rotation="0,0,0", Scale="1,1,1") {
    scaleMode=BMM_Additive
    scaleSpace=BCS_BoneSpace
    translationMode=BMM_Additive
    translationSpace=BCS_BoneSpace
    boneToModify {boneName=wires_cor_02}
    translation {y=0.144, z=-0.518}
   }
   n6 = Anim Dynamics - Bone: tape1!(SimulationSpace:AnimPhysSimSpaceType Enum, ComponentPose=@n5, Alpha=1, GravityOverride:Vector="0,0,-980", ExternalForce="0,0,0") -> (Pose) {
    angularDampingOverride=0.9999
    angularSpringConstant=20
    bAngularSpring=true
    bOverrideAngularDamping=true
    bUseGravityOverride=true
    alphaBoolBlend {blendOption=Linear}
    boundBone {boneName=tape1}
    componentAppliedLinearAccClamp=<3000 3000 2500>
    componentLinearAccScale=<0.5 0.5 0.5>
    componentLinearVelScale=<0.5 0.5 0.5>
    physicsBodyDefinitions [{
      boundBone {boneName=tape1}
      boxExtents=<6 6 6>
      constraintSetup {coneAngle=1, angularLimitsMax=<2 35 1>, angularLimitsMin=<-2 -30 0>, angularTarget=<0 0 0>}
      localJointOffset=<1 0 0>
     }]
    simSpaceSettings {maxAngularAcceleration=3000, maxAngularVelocity=3000, simSpaceAngularAlpha=0.1}
   }
   n7 = Anim Dynamics - Bone: wires_up!(SimulationSpace:AnimPhysSimSpaceType Enum, ComponentPose=@n6, Alpha=1, GravityOverride:Vector="0,0,-980", ExternalForce="0,0,0") -> (Pose) {
    angularDampingOverride=1
    angularSpringConstant=50
    bAngularSpring=true
    bLinearSpring=true
    bOverrideAngularDamping=true
    bOverrideLinearDamping=true
    bUseGravityOverride=true
    linearDampingOverride=1
    linearSpringConstant=50
    alphaBoolBlend {blendOption=Linear}
    boundBone {boneName=wires_up}
    componentAppliedLinearAccClamp=<3000 3000 2500>
    componentLinearAccScale=<0.4 0.18 0.4>
    componentLinearVelScale=<0.18 0.18 0.18>
    physicsBodyDefinitions [{
      boundBone {boneName=wires_up}
      boxExtents=<14 14 14>
      constraintSetup {coneAngle=1, angularLimitsMax=<0 3 0>, angularLimitsMin=<0 0 -3>, angularTarget=<0 0 0>, linearAxesMax=<4 0 0>, linearAxesMin=<0 -0.5 0>}
      localJointOffset=<1 0 0>
     }]
    simSpaceSettings {maxAngularAcceleration=3000, maxAngularVelocity=3000, simSpaceAngularAlpha=0.1}
   }
   n8 = Anim Dynamics - Bone: wires_dwn!(SimulationSpace:AnimPhysSimSpaceType Enum, ComponentPose=@n7, Alpha=1, GravityOverride:Vector="0,0,-980", ExternalForce="0,0,0") -> (Pose) {
    angularDampingOverride=0.9
    angularSpringConstant=3000
    bAngularSpring=true
    bLinearSpring=true
    bOverrideAngularDamping=true
    bOverrideLinearDamping=true
    bUseGravityOverride=true
    linearDampingOverride=0.9999
    linearSpringConstant=50
    alphaBoolBlend {blendOption=Linear}
    boundBone {boneName=wires_dwn}
    componentAppliedLinearAccClamp=<3000 3000 2500>
    componentLinearAccScale=<0.4 0.4 0.4>
    componentLinearVelScale=<0.3 0.3 0.3>
    physicsBodyDefinitions [{
      boundBone {boneName=wires_dwn}
      boxExtents=<15 15 15>
      constraintSetup {coneAngle=1, angularLimitsMax=<0 3 0>, angularLimitsMin=<0 0 -8>, angularTarget=<0 0 0>, linearAxesMax=<5.5 0 0.3>, linearAxesMin=<0 -0.6 0>}
      localJointOffset=<1 0 0>
     }]
    simSpaceSettings {maxAngularAcceleration=3000, maxAngularVelocity=3000, simSpaceAngularAlpha=0.1}
   }
   Output Pose(Result=Component To Local(ComponentPose=@n8)) {layerGroup=None}
  ]}
 EventGraph {nodes [
   TryGetPawnOwner()
   Event BlueprintUpdateAnimation()
  ]}
]
variables []
